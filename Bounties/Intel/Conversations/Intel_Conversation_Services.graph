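# Conversation database for the intel service where NPCs offer to reveal nearby treasure chests.
# The entry tag leads to a speech node with per-speaker lines, gated on the TreasureChest intel service and child requirements.
# That node branches to Back or GrantIntel, which finds unsearched common treasure chest containers through a tag query.
# Search radius, count, rarity, marker and squad visibility come from BountiesGameData curves, with price from CashPricing.
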ConversationDatabase /Bounties/Intel/Conversations/Intel_Conversation_Services
entry_points [{tag=Conversation.Entry.Intel, destinations [#1]}]
nodes [
 ConversationEntryPointNode_0 {
  id=#1
  class=ConversationEntryPointNode
  outputs [#2]
  properties {entryTag {tagName=Conversation.Entry.Intel}}
 }
 Speech_C_1 {
  id=#2
  class=Speech_C
  outputs [#3, #4]
  properties {
   bUsePerNPC_ContextualMessageLookup=true
   softTaskIcon=/Game/Athena/HUD/Interaction/T_UI_NPC_Intel_Chest
   generalConfig {defaultMessage=Better grab it quick.}
   speakerEntryTagToConfig {
    "(TagName=\"Conversation.Entry.Blackheart\")" {defaultMessage="Yar, if it's treasure ye want..."}
    "(TagName=\"Conversation.Entry.MusketSlinger\")" {defaultMessage="Leaving valuables out in the open? Yes, please."}
    "(TagName=\"Conversation.Entry.Temple\")" {defaultMessage=I found a map that told me of this chest's location.}
    "(TagName=\"Conversation.Entry.WreckRaider\")" {defaultMessage=Spotted this earlier. It's yours if you get there.}
   }
  }
  sub_nodes [
   FortConversationRequirement_HasService_5 {class=FortConversationRequirement_HasService, properties {serviceTag {tagName=Tandem.Service.Intel.TreasureChest}}}
   FortConversationChoiceWithPreviewNode_6 {class=FortConversationChoiceWithPreviewNode, properties {defaultChoiceDisplayText="Reveal Nearby Treasure  Chests"}}
   FortConversationRequirement_ChildRequirements_1 {class=FortConversationRequirement_ChildRequirements}
  ]
 }
 GrantIntel_C_0 {
  id=#4
  class=GrantIntel_C
  outputs []
  properties {
   analyticsTag {tagName=Analytics.Intel.TreasureChest.Common}
   filterDatas [{
     classFilter=/Script/FortniteGame.BuildingContainer
     maxRadius {registryType=None, value=1, curve {curveTable=/Bounties/DataTables/BountiesGameData, rowName=Default.Bounties.Intel.TreasureChest.MaxRadius}}
     maxToFind {registryType=None, value=1, curve {curveTable=&BountiesGameData, rowName=Default.Bounties.Intel.TreasureChest.MaxToFind}}
     objectTypes [ObjectTypeQuery1]
     price {registryType=None, value=1, curve {curveTable=/Game/Athena/Balance/DataTables/CashPricing, rowName=Intel}}
     rarity {registryType=None, value=1, curve {curveTable=&BountiesGameData, rowName=Default.Bounties.Intel.TreasureChest.Rarity}}
     showMarkerDetails {
      registryType=None
      value=1
      curve {curveTable=&BountiesGameData, rowName=Default.Bounties.Intel.TreasureChest.ShowMarkerDetails}
     }
     showToSquad {
      registryType=None
      value=1
      curve {curveTable=&BountiesGameData, rowName=Default.Bounties.Intel.TreasureChest.ShowToSquad}
     }
     targetFilter {
      autoDescription=" ALL(  ALL( Building.Type.Container.TreasureChest.Common ),  NONE( Building.Type.Container.AlreadySearched ) )"
      tokenStreamVersion=0
      userDescription=""
      queryTokenStream [0,1,5,2,2,1,0,3,1,1]
      tagDictionary [{tagName=Building.Type.Container.TreasureChest.Common}, {tagName=Building.Type.Container.AlreadySearched}]
     }
    }]
  }
  sub_nodes [ConversationChoiceNode_0 {class=ConversationChoiceNode, properties {defaultChoiceDisplayText=Treasure Chests}}]
 }
 Back_C_4 {
  id=#3
  class=Back_C
  sub_nodes [ConversationChoiceNode_7 {class=ConversationChoiceNode, properties {defaultChoiceDisplayText=BACK}}]
 }
]
